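; Weapon Blueprint for the HighTower Date Combo gauntlet, adding cosmetic overheat feedback on top of its parent weapon.
; Firing starts a looping timer that maps heat to muzzle glow opacity and smoothed, inverted gauntlet glow, cleared on local unequip or cooldown past a threshold flag.
; Fire audio picks first- or third-person sounds, feeding computed shots remaining to the local viewer's sounds, including a low-shot warning.
; Overheating plays its own sound.
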(event FortWeapon|EventOnEquippedAndAttached
  ;[1 1 1] OnEquip - Reset MuzzleGlowMID Opacity
  (|CallParentFunction)
  (Networking|CanExecuteCosmeticEvents
    (:True
      (Rendering|Material|SetScalarParameterValue (Variables|Default|GetMuzzleGlowMID) "Opacity" 0.0 :!class /Script/Engine.MaterialInstanceDynamic))))

(event EventOnPlayWeaponFireFX
  (params (bPersistentFire :type Boolean) (bSecondaryFire :type Boolean))
  (|CallParentFunction bPersistentFire bSecondaryFire)
  (bind _returnvalue (Utilities|Time|SetTimerbyEvent (EventDispatchers|CreateEvent :!event Custom|BarrelOverheatFXandSound) 0.0333 true true))
  (Variables|Default|SetGlowTimer _returnvalue))

(event EventPlayWeaponFireCustomAudio
  (params (bPersistentFire :type Boolean) (bSecondaryFire :type Boolean))
  (bind _returnvalue (Game|GetInstigator))
  (bind _returnvalue_2 (Pawn|IsLocallyViewed _returnvalue))
  (Utilities|FlowControl|Sequence
    (:then_0
      (Variables|Audio|SetShots%20Remaining (Default|GetShotsRemaining)))
    (:then_1
      (Utilities|FlowControl|Sequence
        (:then_0
          ; Play fire sound, the volume is ducked on the last three shots
          (bind _returnvalue_3 (Audio|SpawnSoundAttached (select _returnvalue_2 (Variables|Default|GetFire%201P%20Sound) (Variables|Default|GetFire%203P%20Sound)) (Variables|Visual|GetWeaponMesh)))
          (if _returnvalue_2
            (Audio|Parameter|SetIntegerParameter _returnvalue_3 "ShotsLeft" (Variables|Audio|GetShots%20Remaining) :!class /Script/AudioExtensions.AudioParameterControllerInterface)
            ; Play overheat warning sound on the last three shots
            (if (< (Variables|Audio|GetShots%20Remaining) 3)
              (bind _returnvalue_4 (Audio|SpawnSoundAttached (Variables|Audio|GetOverheat%20Warning%20Sound) (Variables|Visual|GetWeaponMesh)))
              (Audio|Parameter|SetIntegerParameter _returnvalue_4 "ShotsLeft" (Variables|Audio|GetShots%20Remaining) :!class /Script/AudioExtensions.AudioParameterControllerInterface))))))))

(event EventOnUnableToPerformAction
  (params (FailedReason :type Gameplay%20Tag%20Container%20Structure%20%28by%20ref%29) (LocalizedFailureReasonText :type Text%20%28by%20ref%29)))

(event Overheating|EventOnStartOverheated
  ;[1 1 1] Overheat Started
  (|CallParentFunction)
  (Development|PrintString "OVERHEATING")
  (Networking|CanExecuteCosmeticEvents
    (:True
      (Audio|SpawnSoundAttached (Variables|Audio|GetOverheatedSound) (Variables|/Script/Engine.Actor|GetRootComponent (Game|GetInstigator))))))

(event Custom|BarrelOverheatFXandSound
  (bind _returnvalue (Default|CurrentScalarValue))
  (bind _returnvalue_2 (Math|Float|Lerp 0.0 1.0 _returnvalue))
  ;[1 1 1] BarrelGlow - Handles ramping overheat FX and sound.  Timer shuts off when there is no heat in the minigun
  (if (> (Default|CurrentScalarValue) 0.0)
    (Networking|CanExecuteCosmeticEvents
      (:True
        (Rendering|Material|SetScalarParameterValue (Variables|Default|GetMuzzleGlowMID) "Opacity" _returnvalue_2 :!class /Script/Engine.MaterialInstanceDynamic)
        (Variables|Default|SetOverheatSmoothed (Math|Interpolation|FInterpTo (Variables|Default|GetOverheatSmoothed) _returnvalue_2 (Utilities|Time|GetWorldDeltaSeconds) 1e+01))
        (Rendering|Material|SetScalarParameterValueonMaterials (CallFunction|GetGauntletSkelMesh) "GlowActive" (- 1.0 (Variables|Default|GetOverheatSmoothed)))))
    (elif (Variables|Default|GetHitHeatThreshold)
      (Utilities|Time|ClearandInvalidateTimerbyHandle (Variables|Default|GetGlowTimer)))))

(event FortWeapon|EventOnLocalUnEquip
  ;[1 1 1] Clear Timer Reset MuzzleGlowMID Opacity/Sound.  Keeps a phantom sound from playing.
  (|CallParentFunction)
  (Utilities|Time|ClearandInvalidateTimerbyHandle (Variables|Default|GetGlowTimer))
  (Networking|CanExecuteCosmeticEvents
    (:True
      (Rendering|Material|SetScalarParameterValue (Variables|Default|GetMuzzleGlowMID) "Opacity" 0.0 :!class /Script/Engine.MaterialInstanceDynamic))))

(fn CurrentScalarValue
  (return (/ (Overheat|GetCurrentOverheatValue) (Overheat|GetOverheatingMaxValue))))

(fn GetShotsRemaining
  (return (Math|Float|Truncate (/ (- 1.0 (/ (Overheat|GetCurrentOverheatValue) (Overheat|GetOverheatingMaxValue))) (/ (Overheat|GetOverheatHeatingValue) 1e+02)))))

(fn ConstructionScript
  (|CallParentFunction))
